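# Sound cue for the Cube event's angry zap, layering an electrical zap with a delayed Cube voice sound.
# A mixer plays one of two random zap waves at once and one of two random PrintVoice waves after a short random delay.
# A modulator feeds only the first two quality level slots, leaving the others empty.
# Cube impact attenuation applies, and a concurrency override caps simultaneous instances at two, stopping the oldest.
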SoundCue /FNGameplayCues/Athena/Sounds/Events/Cube/Cube_AngryZap_Source_Cue
root=QualityLevel_0
nodes [
 {id=Delay_0, children [Random_1], properties {delayMax=1, delayMin=0.5}}
 {id=Mixer_0, children [Random_0,Delay_0], properties {inputVolume [1,1]}}
 {id=Modulator_4, children [Mixer_0]}
 {id=QualityLevel_0, children [Modulator_4,Modulator_4,null,null]}
 {id=Random_0, children [WavePlayer_0,WavePlayer_2], properties {weights [1,1]}}
 {id=Random_1, children [WavePlayer_1,WavePlayer_3], properties {weights [1,1]}}
 {id=WavePlayer_0, properties {soundWaveAssetPtr=/Game/Athena/Sounds/Events/Cube/Cube_AngryZap_02}}
 {id=WavePlayer_1, properties {soundWaveAssetPtr=/Game/Athena/Sounds/Events/Cube/Cube_PrintVoice_02}}
 {id=WavePlayer_2, properties {soundWaveAssetPtr=/Game/Athena/Sounds/Events/Cube/Cube_AngryZap_01}}
 {id=WavePlayer_3, properties {soundWaveAssetPtr=/Game/Athena/Sounds/Events/Cube/Cube_PrintVoice_01}}
]
properties {
 attenuationSettings=/Game/Athena/Sounds/Events/Cube/Cube_Impacts_Att
 bOverrideConcurrency=true
 maxDistance=10000
 soundClassObject=&Fort_Mission_Ingame_Important_PSM
 concurrencyOverrides {maxCount=2, resolutionRule=StopOldest}
}
