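# Sound cue for the Cube event's area-effect status, layering a sustained loop with an apply sound.
# An IsLocal branch picks stereo status-effect attenuation when true, or mono when false or unset.
# Each path mixes the looping wave, enveloped to rise in volume and pitch, with a quieter, randomly pitched-down, non-looping apply wave.
# It plays in the scripted-events sound class.
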SoundCue /FNGameplayCues/Athena/Sounds/Events/Cube/Cube_AreaEffect_Loop_Cue
root=Branch_0
nodes [
 {
  id=Attenuation_0
  children [Mixer_0]
  properties {attenuationSettings=/FNGameplayCues/Sounds/Attenuation/Fort_Ability_Status_Effect_Looping_Stereo_Att}
 }
 {
  id=Attenuation_1
  children [Mixer_1]
  properties {attenuationSettings=/FNGameplayCues/Sounds/Attenuation/Fort_Ability_Status_Effect_Looping_Mono_Att}
 }
 {id=Branch_0, children [Attenuation_0, Attenuation_1, Attenuation_1], properties {boolParameterName=IsLocal}}
 {
  id=Enveloper_0
  children [WavePlayer_4]
  properties {pitchCurve {editorCurveData {keys [<at 0 0.5>,<at 1 0.6166>]}}, volumeCurve {editorCurveData {keys [<at 0 0>,<at 1 1>]}}}
 }
 {
  id=Enveloper_1
  children [WavePlayer_3]
  properties {pitchCurve {editorCurveData {keys [<at 0 0.5>,<at 1 0.6166>]}}, volumeCurve {editorCurveData {keys [<at 0 0>,<at 1 1>]}}}
 }
 {id=Mixer_0, children [Enveloper_0,Modulator_0], properties {inputVolume [1.25,0.35]}}
 {id=Mixer_1, children [Enveloper_1,Modulator_1], properties {inputVolume [1.25,0.35]}}
 {id=Modulator_0, children [WavePlayer_2], properties {pitchMax=0.9, pitchMin=0.75}}
 {id=Modulator_1, children [WavePlayer_1], properties {pitchMax=0.9, pitchMin=0.75}}
 {id=WavePlayer_1, properties {soundWaveAssetPtr=&/Cube_StatusEffectApply_01}}
 {id=WavePlayer_2, properties {soundWaveAssetPtr=&Cube_StatusEffectApply_01}}
 {id=WavePlayer_3, properties {bLooping=true, soundWaveAssetPtr=&/Cube_StatusEffect_Loop_01}}
 {id=WavePlayer_4, properties {bLooping=true, soundWaveAssetPtr=&Cube_StatusEffect_Loop_01}}
]
properties {
 attenuationSettings=&Fort_Ability_Status_Effect_Looping_Stereo_Att
 maxDistance=3500
 soundClassObject=/Game/Sounds/FortSoundClassesAndModes/Fort_Events_Scripted
 volumeMultiplier=0.786
}
