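# Sound cue for the Cube event's hit ripple, playing small impact sounds at reduced volume.
# A Random node picks one of five equally weighted impact waves and feeds a root Modulator.
# It uses shared Cube impact attenuation and an important in-game sound class.
# A concurrency override caps simultaneous instances at three, stopping the oldest.
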SoundCue /FNGameplayCues/Athena/Sounds/Events/Cube/Cube_HitRipple_Cue
root=Modulator_4
nodes [
 {id=Modulator_4, children [Random_1]}
 {id=Random_1, children [WavePlayer_1,WavePlayer_5,WavePlayer_6,WavePlayer_7,WavePlayer_8], properties {weights [1,1,1,1,1]}}
 {id=WavePlayer_1, properties {soundWaveAssetPtr=/Game/Athena/Sounds/Events/Cube/Cube_ImpactSmall_01}}
 {id=WavePlayer_5, properties {soundWaveAssetPtr=/Game/Athena/Sounds/Events/Cube/Cube_ImpactSmall_02}}
 {id=WavePlayer_6, properties {soundWaveAssetPtr=/Game/Athena/Sounds/Events/Cube/Cube_ImpactSmall_03}}
 {id=WavePlayer_7, properties {soundWaveAssetPtr=/Game/Athena/Sounds/Events/Cube/Cube_ImpactSmall_04}}
 {id=WavePlayer_8, properties {soundWaveAssetPtr=/Game/Athena/Sounds/Events/Cube/Cube_ImpactSmall_05}}
]
properties {
 attenuationSettings=/Game/Athena/Sounds/Events/Cube/Cube_Impacts_Att
 bOverrideConcurrency=true
 maxDistance=10000
 soundClassObject=&Fort_Mission_Ingame_Important_PSM
 volumeMultiplier=0.2
 concurrencyOverrides {maxCount=3, resolutionRule=StopOldest}
}
