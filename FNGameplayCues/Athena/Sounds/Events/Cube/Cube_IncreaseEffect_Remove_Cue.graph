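# Sound cue for removing the Cube "increase effect", filed with the Athena Cube event gameplay cue sounds.
# It has no sound nodes and no root, so the cue plays no audio as it stands.
# Only playback settings remain: status-effect activation attenuation, the scripted-events sound class, a reduced volume and a maximum distance.
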SoundCue /FNGameplayCues/Athena/Sounds/Events/Cube/Cube_IncreaseEffect_Remove_Cue
root=null
nodes []
properties {
 attenuationSettings=/Game/Sounds/Attenuation/Fort_Ability_Status_Effect_Activation_Att
 maxDistance=6500
 soundClassObject=/Game/Sounds/FortSoundClassesAndModes/Fort_Events_Scripted
 volumeMultiplier=0.393
}
