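# Sound cue for the Cube event in Athena, named for the player being launched off the cube's surface, playing an angry zap.
# A Random node picks one of two equally weighted Cube_AngryZap waves and feeds a Modulator at the root, with the cue's pitch lowered.
# It uses the Cube impacts attenuation and the in-game sounds class.
# Its concurrency override allows a single instance and prevents new plays while one is active.
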SoundCue /FNGameplayCues/Athena/Sounds/Events/Cube/Cube_LaunchPlayer_Surface_Cue
root=Modulator_4
nodes [
 {id=Modulator_4, children [Random_1]}
 {id=Random_1, children [WavePlayer_2,WavePlayer_3], properties {weights [1,1]}}
 {id=WavePlayer_2, properties {soundWaveAssetPtr=/Game/Athena/Sounds/Events/Cube/Cube_AngryZap_02}}
 {id=WavePlayer_3, properties {soundWaveAssetPtr=/Game/Athena/Sounds/Events/Cube/Cube_AngryZap_01}}
]
properties {
 attenuationSettings=/Game/Athena/Sounds/Events/Cube/Cube_Impacts_Att
 bOverrideConcurrency=true
 maxDistance=10000
 pitchMultiplier=0.75
 soundClassObject=/Game/Sounds/FortSoundClassesAndModes/Fort_Mission_IngameSounds
 concurrencyOverrides {maxCount=1, resolutionRule=PreventNew}
}
